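# Legacy volumetric cloud material that renders sky clouds and the safe-zone storm ring.
# Puffy clouds come from an SDF volume texture, with drifting noise and a sphere void mask.
# MPC_SafeZoneClouds drives storm placement and size, shaping a profile texture with radial waves, Voronoi noise and rotating 3D flowmaps, plus volume-texture lightning.
# Conservative density and empty-space skipping target raymarch cost, while reflection captures get simplified density without the storm.
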Material /BlastBerryMapContent/Game/TimeOfDay/TODM/MV_RufusCloudAndStormLegacy
graph {
 properties {
  bUsedWithVolumetricCloud=true
  blendMode=BLEND_Additive
  materialDomain=MD_Volume
  pixelDepthOffsetMode=PDOM_Legacy
  usedShadingModels=Default Lit
 }
 nodes [
  n1 = NamedRerouteUsage()
  n2 = ReflectionCapturePassSwitch!(Default=@n1, Reflection=Multiply(A=@n1, B=ScalarParameter() {defaultValue=0.6, parameterName=ReflectionAlbedoBrightness}))
  n3 = ScalarParameter() {
   defaultValue=0
   desc=Rufus addition to help with the non lighted sections fo the storm to make sure the colors and sections stick out
   group=Clouds
   parameterName=Storm_Base_Color_Emissive
  }
  Add(A=NamedRerouteUsage(), B=Multiply(A=Multiply(A=@n2, B=@n3), B=NamedRerouteUsage()))
  n4 = WorldPosition()
  n5 = ScalarParameter() {defaultValue=4500000, group=Storm, parameterName=Scale}
  n6 = Divide(A=@n5)
  n7 = Constant() {r=1}
  ComponentMask(None=Multiply(A=Add(A=@n4.XYZ, B=@n6), B=Divide(A=@n7, B=@n5))) {g=true, r=true}
  n8 = VolumetricAdvancedMaterialInput()
  ComponentMask(None=@n8.ConservativeDensity as Float4) {r=true}
  Constant() {r=1}
  n9 = VectorParameter() {group=Clouds, parameterName=AlbedoClouds, defaultValue=<rgba 1 1 1 0>}
  n10 = NamedRerouteUsage()
  n11 = DotProduct(
   A=TextureSample(UVs=Multiply(A=@n10)) {texture=/Volumetrics/Content/VolumeTextures/Textures/VT_PerlinWorley_Balanced}.RGBA
   B=Constant4Vector() {desc=fbm weights 0.7, constant=<rgba 0.25 0.25 0.25 0.25>}.RGBA
  )
  n12 = Constant() {r=0.6}
  n13 = LinearInterpolate(
   A=VectorParameter() {group=Storm, parameterName=AlbedoStormBottom, defaultValue=<rgba 0.2602 0.1591 0.2656 1>}.RGB
   B=VectorParameter() {group=Storm, parameterName=AlbedoStormTop, defaultValue=<rgba 0 0.0938 0.0649 1>}.RGB
   Alpha=Saturate(None=MaterialFunctionCall(In=@n11, Contrast=@n12) {materialFunction=&CheapContrast}.Result)
  )
  n14 = NamedRerouteUsage()
  LinearInterpolate(A=@n9.RGB, B=@n13, Alpha=@n14)
  LinearInterpolate(A=@n13, B=@n9.RGB, Alpha=NamedRerouteUsage())
  MaterialFunctionCall(A=@n9.RGB, B=NamedRerouteUsage(), C=@n13, Alpha=@n14) {materialFunction=&Lerp_3Color}
  n15 = NamedRerouteUsage()
  MaterialFunctionCall(
   Input=@n15
   Input Low=ScalarParameter() {defaultValue=0.5, parameterName=max}
   Input High=ScalarParameter() {defaultValue=-0.1667, parameterName=min}
  ) {materialFunction=&RemapValue_0-1}
  n16 = Saturate(None=NamedRerouteUsage())
  n17 = Multiply(A=Ceil(None=@n16), B=Ceil(None=OneMinus(None=@n16)))
  MaterialFunctionCall(X=Max(A=NamedRerouteUsage(), B=@n17), Y=NamedRerouteUsage()) {materialFunction=&MakeFloat4}
  n18 = Constant() {r=1}
  MaterialFunctionCall(X=@n18, Y=@n18, Z=Multiply(A=Time(), B=ScalarParameter() {defaultValue=0.001, parameterName=VerticalNioseSpeed}) {constB=0.05}) {materialFunction=&MakeFloat3}
  n19 = NamedRerouteUsage()
  n20 = NamedRerouteUsage()
  n21 = NamedRerouteUsage()
  Max(A=ReflectionCapturePassSwitch(Default=Multiply(A=@n19, B=@n20), Reflection=@n19), B=@n21)
  n22 = Time()
  n23 = CollectionParameter() {collection=/Game/Athena/SafeZone/MPC_SafeZoneClouds, parameterName=CustomCloudsTime}
  n24 = Ceil(None=Saturate(None=@n23))
  Multiply(
   A=NamedRerouteDeclaration(None=LinearInterpolate(A=@n22, B=@n23, Alpha=@n24)) {name=CustomTime, nodeColor=<rgba 0 1 0.1094 1>}
   B=CollectionParameter() {collection=&MPC_SafeZoneClouds, parameterName=TopCloudsSpeed}
  )
  Multiply(A=ComponentMask(None=NamedRerouteUsage()) {b=true}, B=NamedRerouteUsage())
  NamedRerouteDeclaration(None=@n21) {name=density, nodeColor=<rgba 0.9922 0 1 1>}
  NamedRerouteDeclaration(None=@n5) {name=halfScale, nodeColor=<rgba 1 0 0.7266 1>}
  NamedRerouteDeclaration(None=VectorParameter() {parameterName=StormCloudBlendColor, defaultValue=<rgba 0.0417 0.0417 0.0417 1>}.RGB) {name=StormBlend, nodeColor=<rgba 0.9453 1 0 1>}
  n25 = Distance(
   A=ComponentMask(None=WorldPosition().XYZ) {g=true, r=true}
   B=ComponentMask(None=CollectionParameter() {collection=&MPC_SafeZoneClouds, parameterName=SafeZoneLoc}) {g=true, r=true}
  )
  n26 = MaterialFunctionCall(Float2=NamedRerouteUsage()) {materialFunction=&BreakOutFloat2Components}
  n27 = MaterialFunctionCall!(Km=@n26.G) {materialFunction=/Engine/Functions/Engine_MaterialFunctions01/Units/Km-to-Cm}
  n28 = ScalarParameter() {defaultValue=1, group=StormRadialWaves, parameterName=WavesInnerRadiusMultiplier}
  n29 = Subtract(A=@n25, B=@n27.Cm)
  n30 = MaterialFunctionCall(Km=@n26.R) {materialFunction=&Km-to-Cm}
  n31 = Add(A=@n30.Cm, B=@n27.Cm)
  n32 = CloudSampleAttribute()
  n33 = Constant()
  n34 = ScalarParameter() {defaultValue=1, parameterName=HighRemap}
  n35 = NamedRerouteUsage()
  n36 = Divide(
   A=Subtract(
    A=MaterialFunctionCall(Input=@n32.NormAltitudeInLayer, Input Low=@n33, Input High=@n34) {materialFunction=&RemapValue_0-1}.Result
    B=ComponentMask(None=@n35) {b=true}
   )
   B=ComponentMask(None=@n35) {a=true}
  )
  n37 = Add(A=Saturate(None=Divide(A=@n29, B=@n31)), B=Saturate(None=OneMinus(None=@n36)))
  n38 = Add(
   A=MaterialFunctionCall!(Vector or UVs=NamedRerouteUsage()) {materialFunction=/Game/Athena/SafeZone/S11/MF_VectorToRadialValueFast}.Result
   B=Multiply(A=Time(), B=ScalarParameter() {group=StormRadialWaves, parameterName=WavesSpeed})
  )
  n39 = Sine(None=Multiply(A=Frac(None=@n38), B=ScalarParameter() {defaultValue=4, group=StormRadialWaves, parameterName=WaveCount}) {constB=10})
  n40 = Power(
   Base=Add(A=Multiply(A=@n39) {constB=0.5}) {constB=0.5}
   Exp=ScalarParameter() {defaultValue=1, group=StormRadialWaves, parameterName=WavesPow}
  )
  n41 = ComponentMask(None=Saturate(None=Add(A=@n37, B=@n40) {constB=0})) {r=true}
  n42 = OneMinus(None=@n41)
  n43 = Multiply(A=@n27.Cm, B=Saturate(None=Add(A=@n28, B=@n42)))
  NamedRerouteDeclaration(None=Divide(A=Subtract(A=@n25, B=@n43), B=@n30.Cm)) {name=Storm Radius, nodeColor=<rgba 0.0547 1 0 1>}
  n44 = MaterialFunctionCall!(WorldScale=ScalarParameter() {defaultValue=2250000, parameterName=worldScale}) {materialFunction=/Game/TimeOfDay/CloudsAndStorm/Materials/MaterialFunctions/MF_WorldRemap}
  n45 = Multiply(A=ComponentMask(None=NamedRerouteUsage()) {g=true, r=true}, B=ScalarParameter() {defaultValue=0.5, parameterName=SphereMaskErosionGain}) {constB=0.2}
  n46 = SphereMask(
   A=ReflectionCapturePassSwitch(Default=Subtract(A=@n44.Result, B=@n45), Reflection=@n44.Result)
   B=Constant()
   Radius=ScalarParameter() {defaultValue=0.2, parameterName=maskRadius}
   Hardness=ScalarParameter() {defaultValue=0.333, parameterName=MaskHardness}
  )
  NamedRerouteDeclaration(None=Saturate(None=OneMinus(None=@n46))) {name=voidMask, nodeColor=<rgba 0 0.0391 1 1>}
  n47 = NamedRerouteUsage()
  n48 = ScalarParameter() {defaultValue=0.001, group=Storm, parameterName=scc}
  n49 = Saturate(None=Add(A=Multiply(A=@n47, B=@n48), B=ScalarParameter() {defaultValue=0.5, group=Storm, parameterName=off}))
  n50 = Divide(A=LinearInterpolate(
    A=ScalarParameter() {defaultValue=128, group=Storm, parameterName=Global Storm Density}
    B=ScalarParameter() {defaultValue=128, group=Storm, parameterName=Global Storm Density Top}
    Alpha=@n49
   )) {constB=100}
  n51 = NamedRerouteUsage()
  n52 = ScalarParameter() {defaultValue=0.5286, group=StormDFs, parameterName=Noise Intensity}
  n53 = ScalarParameter() {defaultValue=0.2, group=StormDFs, parameterName=Noise IntensityTop}
  n54 = Multiply(A=NamedRerouteUsage(), B=VectorParameter() {group=Storm, parameterName=WorldScaleNoise, defaultValue=<rgba 1 1 1 0>}.RGB)
  n55 = LinearInterpolate(
   A=ScalarParameter() {defaultValue=1, group=Storm, parameterName=NoiseGlobalScaleEnd}
   B=ScalarParameter() {defaultValue=1, group=Storm, parameterName=NoiseGlobalScale}
   Alpha=CollectionParameter() {collection=&MPC_SafeZoneClouds, parameterName=SafeZoneScale}
  )
  n56 = Multiply(A=ScalarParameter() {defaultValue=4500000, group=Storm, parameterName=scale}, B=@n55)
  n57 = TextureSample!(
   UVs=Divide(A=@n54, B=MaterialFunctionCall(Km=@n56) {materialFunction=&Km-to-Cm}.Cm)
   Tex
   Level=CollectionParameter() {collection=&MPC_SafeZoneClouds, parameterName=StormDetailVolumeMip}
   Apply View MipBias
  ) {mipValueMode=TMVM_MipLevel, samplerType=SAMPLERTYPE_LinearColor, texture=/Volumetrics/Content/VolumeTextures/Textures/VT_Voronoi_Simple}
  n58 = ScalarParameter() {defaultValue=0.8, group=StormDFs, parameterName=Noise DF Center End}
  n59 = ScalarParameter() {defaultValue=0.8, group=StormDFs, parameterName="Noise DF Center "}
  n60 = CollectionParameter() {collection=&MPC_SafeZoneClouds, parameterName=SafeZoneScale}
  n61 = Multiply(
   A=Multiply(A=ComponentMask(None=@n51) {r=true}, B=LinearInterpolate(A=@n52, B=@n53, Alpha=@n49))
   B=Subtract(A=@n57.R, B=LinearInterpolate(A=@n58, B=@n59, Alpha=@n60)) {constB=0.5}
  )
  n62 = Constant()
  n63 = MaterialFunctionCall(Float2=Normalize(VectorInput=NamedRerouteUsage())) {materialFunction=&BreakOutFloat2Components}
  n64 = AppendVector(A=Multiply(A=@n63.G) {constB=-1}, B=@n63.R)
  n65 = Constant()
  n66 = Multiply(A=@n55, B=ScalarParameter() {defaultValue=512, group=Storm, parameterName=scale2})
  n67 = MaterialFunctionCall!(
   Volume Texture=TextureObject() {samplerType=SAMPLERTYPE_LinearColor, texture=&VT_Voronoi_Simple}
   UVW=Divide(A=@n54, B=MaterialFunctionCall(Km=@n66) {materialFunction=&Km-to-Cm}.Cm)
   Velocity=Multiply(A=AppendVector(A=@n64, B=@n65), B=ScalarParameter() {defaultValue=0.7, group=StormFlowMap, parameterName=Storm Rotation Velocity})
   Override Time=Multiply(A=ScalarParameter() {defaultValue=0.1, group=StormFlowMap, parameterName=T}, B=NamedRerouteUsage())
   Mip Level=CollectionParameter() {collection=&MPC_SafeZoneClouds, parameterName=StormFlowMapsVolumeMip}
  ) {materialFunction=/Game/TimeOfDay/CloudsAndStorm/Materials/MaterialFunctions/Flowmaps_3D}
  n68 = CollectionParameter() {collection=&MPC_SafeZoneClouds, parameterName=CloudStartingBias}
  n69 = Subtract(
   A=NamedRerouteDeclaration(None=ComponentMask(None=@n67.Result RGB) {r=true}) {name=FlowMap, nodeColor=<rgba 1 0 0.1172 1>}
   B=LinearInterpolate(
    A=ScalarParameter() {defaultValue=0.8, group=StormDFs, parameterName="Flowmaps DF Center "}
    B=ScalarParameter() {defaultValue=1, group=StormDFs, parameterName="Flowmaps DF Center PreStart "}
    Alpha=Saturate(None=@n68)
   )
  ) {constB=0.5}
  n70 = ScalarParameter() {defaultValue=0.5, group=StormDFs, parameterName=Flowmaps Intensity}
  MaterialExpressionNamedRerouteUsage_4 = NamedRerouteUsage()
  MaterialExpressionComponentMask_15 = ComponentMask(None=NamedRerouteUsage()) {g=true}
  n71 = Custom!(NormAltitude=@MaterialExpressionNamedRerouteUsage_4, Density=@MaterialExpressionComponentMask_15) {
   code="float HeightDensity = Density;\r\n\r\nif(NormAltitude<0.05 || NormAltitude>0.95)\r\n{\r\n   HeightDensity = 0.0;\r\n}\r\n\r\nreturn HeightDensity;"
   inputs [
    {
     inputName=NormAltitude
     input {expression=&self:MaterialExpressionNamedRerouteUsage_4}
    }
    {
     inputName=Density
     input {expression=&self:MaterialExpressionComponentMask_15}
    }
   ]
  }
  n72 = ScalarParameter() {defaultValue=0.681, group=StormDFs, parameterName=StormDonut DF Center}
  n73 = Add(
   A=Add(A=ReflectionCapturePassSwitch(Default=@n61, Reflection=@n62), B=Multiply(A=@n69, B=@n70))
   B=Multiply(A=ScalarParameter() {defaultValue=0.5286, group=StormDFs, parameterName=StormDonut Intensity}, B=Subtract(A=@n71, B=@n72))
  )
  n74 = OneMinus(None=Max(
    A=CollectionParameter() {collection=&MPC_SafeZoneClouds, parameterName=CloudStartingBias}
    B=CollectionParameter() {collection=&MPC_SafeZoneClouds, parameterName=PreventStormShrinking}
   ))
  n75 = ReflectionCapturePassSwitch(
   Default=Multiply(A=CollectionParameter() {collection=&MPC_SafeZoneClouds, parameterName=EnableVolumetricStorm}, B=Saturate(None=@n74))
   Reflection=Constant()
  )
  NamedRerouteDeclaration(None=Multiply(A=Multiply(A=@n50, B=@n73), B=@n75)) {name=StormEXT, nodeColor=<rgba 1 0 0.8894 1>}
  n76 = CollectionParameter() {collection=&MPC_SafeZoneClouds, parameterName=SafeZoneLoc}
  NamedRerouteDeclaration(None=Subtract(A=WorldPosition().XYZ, B=ComponentMask(None=@n76) {b=true, g=true, r=true})) {name=Safe Zone Local, nodeColor=<rgba 1 0 0.7266 1>}
  n77 = ComponentMask(None=@n8.ConservativeDensity as Float4) {g=true}
  n78 = ScalarParameter() {defaultValue=0.5, parameterName=max}
  n79 = ScalarParameter() {defaultValue=-0.1667, parameterName=min}
  n80 = MaterialFunctionCall(Input=Subtract(A=@n77, B=NamedRerouteUsage()), Input Low=@n78, Input High=@n79) {materialFunction=&RemapValue_0-1}
  n81 = ScalarParameter() {defaultValue=0.064, parameterName=density}
  n82 = MaterialFunctionCall(Input=@n77, Input Low=@n78, Input High=@n79) {materialFunction=&RemapValue_0-1}
  n83 = ScalarParameter() {defaultValue=0.064, parameterName=reflectionDensity}
  NamedRerouteDeclaration(None=ReflectionCapturePassSwitch(Default=Multiply(A=@n80.Result, B=@n81) {constB=0.0001}, Reflection=Multiply(A=@n82.Result, B=@n83) {constB=0.0001})) {name=puffyCloudDensity, nodeColor=<rgba 0 1 0.4844 1>}
  n84 = WorldPosition()
  n85 = MaterialFunctionCall(Km=ScalarParameter() {defaultValue=5, group=Lightning, parameterName=Lighting Pattern Scale}) {materialFunction=&Km-to-Cm}
  n86 = AppendVector(
   A=ComponentMask(None=Divide(A=@n84.XYZ, B=@n85.Cm)) {g=true, r=true}
   B=MaterialFunctionCall(Speed=ScalarParameter() {defaultValue=0.2, group=Lightning, parameterName=Lighting Speed}) {materialFunction=&TimeWithSpeedVariable}.Time
  )
  n87 = TextureSample!(UVs=@n86, Tex, Apply View MipBias) {texture=/Volumetrics/Content/VolumeTextures/Textures/VT_Lightning}
  n88 = Add(
   A=Power(Base=@n87.R, Exp=ScalarParameter() {defaultValue=3, group=Lightning, parameterName=Lightning Power})
   B=Multiply(A=Multiply(A=@n87.R, B=@n87.R), B=ScalarParameter() {group=Lightning, parameterName=Fill Lightning Scatter})
  )
  n89 = Multiply(
   A=TextureSample!(UVs=@n86, Tex, Level=ScalarParameter() {defaultValue=2.5, group=Lightning, parameterName=Lightning Mip}, Apply View MipBias) {mipValueMode=TMVM_MipLevel, texture=&VT_Lightning}.R
   B=ScalarParameter() {defaultValue=0.05, group=Lightning, parameterName=Lightning Spread}
  )
  n90 = NamedRerouteUsage()
  n91 = ScalarParameter() {defaultValue=1, group=Lightning, parameterName=Lightning Mask Power}
  n92 = Multiply(
   A=Multiply(A=Add(A=@n88, B=@n89) {constB=0}, B=Power(Base=@n90, Exp=@n91))
   B=VectorParameter() {group=Lightning, parameterName=Lightning Color, defaultValue=<rgba 0.0997 0.3491 2 0>}.RGB
  )
  NamedRerouteDeclaration(None=@n92) {name=Lightning, nodeColor=<rgba 1 0 0.2223 1>}
  NamedRerouteDeclaration(None=@n75) {name=Storm Control, nodeColor=<rgba 1 0 0.7734 1>}
  n93 = NamedRerouteUsage()
  n94 = NamedRerouteUsage()
  n95 = TextureSample(
   UVs=MaterialFunctionCall(X=Saturate(None=@n93), Y=OneMinus(None=@n94)) {materialFunction=&MakeFloat2}.Result
   Level=CollectionParameter() {collection=&MPC_SafeZoneClouds, parameterName=StormProfileMip}
  ) {mipValueMode=TMVM_MipLevel, samplerType=SAMPLERTYPE_LinearColor, texture=/Game/TimeOfDay/CloudsAndStorm/Textures/StormProfileLegacy}
  NamedRerouteDeclaration(None=@n95.A) {name=StormShape, nodeColor=<rgba 1 0.141 0.7007 1>}
  n96 = CloudSampleAttribute()
  n97 = ScalarParameter() {parameterName=LowAltRemapCol}
  n98 = ScalarParameter() {defaultValue=1, parameterName=HIghAltRemapCol}
  n99 = OneMinus(None=Saturate(None=MaterialFunctionCall(Input=@n96.NormAltitudeInLayer, Input Low=@n97, Input High=@n98) {materialFunction=&RemapValueRange}.Result))
  NamedRerouteDeclaration(None=OneMinus(None=Power(Base=@n99))) {name=rampColour, nodeColor=<rgba 0.6875 1 0 1>}
  n100 = NamedRerouteUsage()
  n101 = Subtract(
   A=MaterialFunctionCall(
    Input=CloudSampleAttribute().NormAltitudeInLayer
    Input Low=ScalarParameter() {defaultValue=-0.02, parameterName=LowRemap}
    Input High=ScalarParameter() {defaultValue=1, parameterName=HighRemap}
   ) {materialFunction=&RemapValue_0-1}.Result
   B=ComponentMask(None=@n100) {b=true}
  )
  n102 = Add(
   A=ComponentMask(None=@n100) {a=true}
   B=Multiply(A=ScalarParameter() {defaultValue=-0.0857, group=StormRadialWaves, parameterName=WavesHeight}, B=@n41)
  )
  NamedRerouteDeclaration(None=Saturate(None=Divide(A=@n101, B=@n102))) {name=Linear Height Profile, nodeColor=<rgba 1 0.1875 0 1>}
  n103 = Multiply(A=NamedRerouteUsage(), B=ScalarParameter() {defaultValue=26, parameterName=offsetNoiseFrequencyb})
  n104 = Multiply(A=NamedRerouteUsage(), B=ScalarParameter() {defaultValue=0.01, desc=0.002, parameterName=BaseCloudSpeedMult})
  n105 = ConstantBiasScale(None=TextureSample(UVs=Add(A=@n103, B=@n104)) {mipValueMode=TMVM_MipLevel, samplerType=SAMPLERTYPE_LinearColor, texture=/Game/Global/MaterialFunctions/VT_ClothSimWPO}.R) {bias=-0.5, scale=2}
  n106 = MaterialFunctionCall(
   Input=CloudSampleAttribute().NormAltitudeInLayer
   Input Low=ScalarParameter() {defaultValue=-0.2, parameterName=LowAltRemap}
   Input High=ScalarParameter() {defaultValue=0.5, parameterName=HIghAltRemap}
  ) {materialFunction=&RemapValue_0-1}
  NamedRerouteDeclaration(None=Multiply(A=Multiply(A=@n105, B=@n106.Result), B=ScalarParameter() {defaultValue=0.1, desc=0.1, parameterName=NoiseGain})) {name=CloudNoise, nodeColor=<rgba 0 0.6953 1 1>}
  NamedRerouteDeclaration(None=@n95.RGB) {name=Mapped Storm Profile, nodeColor=<rgba 0.8984 0 1 1>}
  n107 = Time()
  NamedRerouteDeclaration(None=MaterialFunctionCall(X=@n107, Y=@n107, Z=Constant()) {materialFunction=&MakeFloat3}.Result) {name=opposingWind, nodeColor=<rgba 0.6172 0 1 1>}
  n108 = NamedRerouteUsage()
  n109 = MaterialFunctionCall(Input=CloudSampleAttribute().NormAltitudeInLayer, Input Low=NamedRerouteUsage(), Input High=NamedRerouteUsage()) {materialFunction=&RemapValueRange}
  n110 = TextureSample!(UVs=AppendVector(A=ComponentMask(None=@n108) {g=true, r=true}, B=Saturate(None=@n109.Result)), Tex, Apply View MipBias) {
   samplerSource=SSM_Wrap_WorldGroupSettings
   samplerType=SAMPLERTYPE_LinearColor
   texture=/Game/TimeOfDay/CloudsAndStorm/Textures/VolumeCloudSDF_v04_Volume
  }
  NamedRerouteDeclaration(None=@n110.R) {name=RawCloudSDF, nodeColor=<rgba 1 0.5391 0 1>}
  n111 = Multiply(A=Subtract(A=@n15, B=ScalarParameter() {defaultValue=0.6, parameterName=CloudConservativeExpand})) {constB=-1}
  NamedRerouteDeclaration(None=Saturate(None=Sign(None=@n111))) {name=ConservativeClouds, nodeColor=<rgba 0 0.0156 1 1>}
  NamedRerouteDeclaration(None=@n17) {name=ConservativeVector, nodeColor=<rgba 0 0.8594 1 1>}
  n112 = NamedRerouteUsage()
  n113 = Multiply(A=Add(A=WorldPosition().XYZ, B=Multiply(A=@n112) {constB=0.5}), B=Divide(A=Constant() {r=1}, B=@n112))
  n114 = Multiply(A=NamedRerouteUsage(), B=NamedRerouteUsage())
  n115 = Multiply(A=ComponentMask(None=@n114) {r=true}) {constB=0.6667}
  NamedRerouteDeclaration(None=Subtract(
    A=@n113
    B=MaterialFunctionCall(X=@n114, Y=@n114, Z=@n115) {materialFunction=&MakeFloat3}.Result
   )) {name=worldUV, nodeColor=<rgba 0 0.7188 1 1>}
  NamedRerouteDeclaration(None=ScalarParameter() {defaultValue=2250000, parameterName=worldScale}) {name=Scale, nodeColor=<rgba 1 0 0.8203 1>}
  n116 = CollectionParameter() {collection=&MPC_SafeZoneClouds, parameterName=StormThicknessKm-End}
  n117 = CollectionParameter() {collection=&MPC_SafeZoneClouds, parameterName=StormRadiusKm-End}
  n118 = CollectionParameter() {collection=&MPC_SafeZoneClouds, parameterName=StormAltitude-End}
  n119 = CollectionParameter() {collection=&MPC_SafeZoneClouds, parameterName=StormHeight-End}
  n120 = CollectionParameter() {collection=&MPC_SafeZoneClouds, parameterName=StormThicknessKm}
  n121 = Saturate(None=CollectionParameter() {collection=&MPC_SafeZoneClouds, parameterName=CloudStartingBias})
  n122 = Add(
   A=CollectionParameter() {collection=&MPC_SafeZoneClouds, parameterName=StormRadiusKm}
   B=Multiply(A=ScalarParameter() {defaultValue=3, parameterName=StormRadius_PreStartOffset}, B=Multiply(A=@n121, B=@n121))
  )
  n123 = CollectionParameter() {collection=&MPC_SafeZoneClouds, parameterName=StormAltitude}
  n124 = CollectionParameter() {collection=&MPC_SafeZoneClouds, parameterName=StormHeight}
  NamedRerouteDeclaration(None=LinearInterpolate(
    A=MaterialFunctionCall(X=@n116, Y=@n117, Z=@n118, A=@n119) {materialFunction=&MakeFloat4}.Result
    B=MaterialFunctionCall(X=@n120, Y=@n122, Z=@n123, A=@n124) {materialFunction=&MakeFloat4}.Result
    Alpha=CollectionParameter() {collection=&MPC_SafeZoneClouds, parameterName=SafeZoneScale}
   )) {name=Storm Collection Params, nodeColor=<rgba 0.4708 0 1 1>}
  NamedRerouteDeclaration(None=ScalarParameter() {defaultValue=-0.1, parameterName=LowAltRemapAlt}) {name=LowRemap, nodeColor=<rgba 0.4297 0 1 1>}
  NamedRerouteDeclaration(None=ScalarParameter() {defaultValue=1.5, parameterName=HIghAltRemapAlt}) {name=HighRemap, nodeColor=<rgba 0.9688 0 1 1>}
  NamedRerouteDeclaration(None=@n67.UVW_A) {name=FlowUVs, nodeColor=<rgba 0 0.5312 1 1>}
  NamedRerouteDeclaration(None=ScalarParameter() {defaultValue=0.001, desc=0.002, parameterName=BaseCloudSpeed}) {name=CloudSpeed, nodeColor=<rgba 1 0.4453 0 1>}
  NamedRerouteDeclaration(None=@n13) {name=CloudsAlbedo, nodeColor=<rgba 0 0.625 1 1>}
  NamedRerouteDeclaration(None=@n113) {name=worldUVStormStatic, nodeColor=<rgba 0 0.9062 1 1>}
  NamedRerouteDeclaration(None=@n44.Result) {name=UVBOX, nodeColor=<rgba 0 1 0.8125 1>}
  NamedRerouteDeclaration(None=Multiply(A=NamedRerouteUsage(), B=NamedRerouteUsage())) {name=CloudsEmissive, nodeColor=<rgba 0.8516 1 0 1>}
  n125 = ComponentMask(None=VolumetricAdvancedMaterialInput().ConservativeDensity as Float4) {g=true}
  n126 = ScalarParameter() {
   defaultValue=0.5
   desc="In SDF space. how far away from the cloud surface, the colour starts to belnd from white to storm"
   parameterName=CloudStormColorBlendStartDistance
  }
  n127 = Add(A=@n126, B=ScalarParameter() {defaultValue=0.3, parameterName=CloudStormColorBlendWidth})
  NamedRerouteDeclaration(None=Power(Base=MaterialFunctionCall(Input=@n125, Input Low=@n126, Input High=@n127) {materialFunction=&RemapValue_0-1}.Result) {constExponent=0.4}) {name=StormCloudBlend, nodeColor=<rgba 0 0.2266 1 1>}
  NamedRerouteUsage()
  NamedRerouteUsage()
  NamedRerouteUsage()
  NamedRerouteUsage()
  ReflectionCapturePassSwitch(Default=Constant(), Reflection=Constant() {r=8})
  ScalarParameter() {defaultValue=4000, desc=3000.0, parameterName=NoiseSpeed}
  ScalarParameter() {defaultValue=0.5, parameterName=AO}
  StaticSwitchParameter(True=ReflectionCapturePassSwitch(Default=@n13, Reflection=@n13)) {group=Storm, parameterName=EnableStorm}
  StaticSwitchParameter(True=Multiply(A=OneMinus(), B=@n92), False=@n92) {group=Skyfire, parameterName=UseSkyfire, sortPriority=-1}
  Subtract(A=@n125, B=@n126)
  VolumetricAdvancedMaterialOutput!(
   PhaseG=ScalarParameter() {defaultValue=-0.1, group=CloudProperties, parameterName=PhaseG}
   PhaseG2=ScalarParameter() {defaultValue=0.7, group=CloudProperties, parameterName=PhaseG2}
   PhaseBlend=ScalarParameter() {defaultValue=0.5, group=CloudProperties, parameterName=PhaseBlend}
   MultiScatteringContribution=ScalarParameter() {defaultValue=1, group=CloudProperties, parameterName=MultiScatteringContribution}
   MultiScatteringOcclusion=ScalarParameter() {defaultValue=0.2, group=CloudProperties, parameterName=MultiScatteringOcclusion}
   MultiScatteringEccentricity=ScalarParameter() {defaultValue=0.5, group=CloudProperties, parameterName=MultiScatteringEccentricity}
   ConservativeDensity=NamedRerouteUsage()
  ) {multiScatteringApproximationOctaveCount=1}
  n128 = ScalarParameter() {defaultValue=0.7, parameterName=CloudConservativeExpandSkip}
  n129 = Multiply(A=Multiply(A=Subtract(A=@n15, B=@n128)) {constB=-1}) {constB=100}
  VolumetricCloudEmptySpaceSkippingOutput!(ContainsMatter=Saturate(None=Sign(None=@n129)))
 ]
}
